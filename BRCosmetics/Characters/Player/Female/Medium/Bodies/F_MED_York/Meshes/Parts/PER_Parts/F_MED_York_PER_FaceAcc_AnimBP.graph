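# Animation Blueprint for the York PER female face-accessory part that follows the attached parent mesh and adds secondary physics.
# Copied pose passes through two additive bone-space offsets on cau_hair_lod_correct.
# A RigidBody node then simulates relative to neck_01 with damped component motion, borrowing the Elastic CAU physics asset.
# The event graph sets LOD 1 and LOD 2 flags, but the AnimGraph shows no link to them.
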AnimBlueprint /BRCosmetics/Characters/Player/Female/Medium/Bodies/F_MED_York/Meshes/Parts/PER_Parts/F_MED_York_PER_FaceAcc_AnimBP
target_skeleton=&/F_MED_York_PER_FaceAcc_Skeleton
graphs [
 AnimGraph {nodes [
   n1 = "Transform (Modify) Bone - Bone: cau_hair_lod_correct"!(
    ComponentPose=Local To Component(LocalPose=Copy Pose From Mesh() {bUseAttachedParent=true})
    bAlphaBoolEnabled:bool=True
    Translation="1,0,0"
    Rotation="0,0,0"
    Scale="0.997,1.000,1.000"
   ) -> (Pose) {
    alphaInputType=Bool
    scaleMode=BMM_Additive
    scaleSpace=BCS_BoneSpace
    translationMode=BMM_Additive
    translationSpace=BCS_BoneSpace
    alphaBoolBlend {blendInTime=0.2, blendOption=Linear, blendOutTime=0.2}
    boneToModify {boneName=cau_hair_lod_correct}
    scale {x=0.9967}
    translation {x=1}
   }
   n2 = "Transform (Modify) Bone - Bone: cau_hair_lod_correct"(ComponentPose=@n1, bAlphaBoolEnabled=True, Translation="2,0,0", Rotation="0,0,0", Scale="0.997,1.000,1.000") {
    alphaInputType=Bool
    scaleMode=BMM_Additive
    scaleSpace=BCS_BoneSpace
    translationMode=BMM_Additive
    translationSpace=BCS_BoneSpace
    alphaBoolBlend {blendInTime=0.2, blendOption=Linear, blendOutTime=0.2}
    boneToModify {boneName=cau_hair_lod_correct}
    scale {x=0.9967}
    translation {x=2}
   }
   n3 = RigidBody(ComponentPose=@n2) {
    bClampLinearTranslationLimitToRefPose=true
    bOverrideWorldGravity=true
    overridePhysicsAsset=/BRCosmetics/Characters/Player/Female/Medium/Bodies/F_MED_Elastic/Meshes/Parts/CAU_Parts/F_MED_Elastic_CAU_1_Physics
    simulationSpace=BaseBoneSpace
    alphaBoolBlend {blendOption=Linear}
    baseBoneRef {boneName=neck_01}
    componentLinearAccScale=<0.3 0.3 0.3>
    componentLinearVelScale=<0.3 0.3 0.3>
    overrideWorldGravity {z=-980}
   }
   Output Pose(Result=Component To Local(ComponentPose=@n3)) {layerGroup=None}
  ]}
 EventGraph {nodes [
   Event BlueprintUpdateAnimation()
   n1 = Get CurrentLOD()
   > Set Is lod 1(Is lod 1:bool="Equal (Integer)"(A=@n1, B=1)) -> (Output_Get:bool)
   > Set is lod 2(is lod 2:bool="Equal (Integer)"(A=@n1, B=2)) -> (Output_Get:bool)
   TryGetPawnOwner()
  ]}
]
variables [
 Is lod 1 {type=boolean}
 is lod 2 {type=boolean}
]
